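; Slab live-event script that drives a POI change, its sequence and building smash, and lethal damage to players.
; On the server, calendar events R0 and R1, gated by a setup flag, set SlabPOIChanged, whose OnRep toggles level hiding.
; Custom events relay sequence cues through dispatcher calls.
; Environmental damage kills pawns in a fixed box trace or overlapping a given actor, the latter skipping DBNO.
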(event Custom|KillPlayersDestroyPBWs
  (bind _outhits (Collision|MultiBoxTraceForObjects :Start "110352.781250,35746.035156,493.366455" :End "110352.781250,35746.035156,493.366455" :HalfSize "5379.943848,5379.943848,927.546387" :Orientation "0, 0, 0" :ObjectTypes (Utilities|Array|MakeArray "ObjectTypeQuery3") :bTraceComplex false :DrawDebugType "None" :bIgnoreSelf true :TraceColor "(R=1.000000,G=0.000000,B=0.000000,A=1.000000)" :TraceHitColor "(R=0.000000,G=1.000000,B=0.000000,A=1.000000)" :DrawTime 2e+01))
  (for _array_element _outhits
    (bind _asplayer_pawn_athena (Utilities|Casting|CastToPlayerPawn_Athena (.HitActor (Collision|BreakHitResult _array_element))))
    (Environment|ApplyGlobalEnvironmentDamagetoActor _asplayer_pawn_athena 1e+03 "(GameplayTags=((TagName=\"Gameplay.Damage.Environment\"),(TagName=\"Gameplay.Effect.InstantDeath.Environment\")))")))

(event Custom|PlaySlab
  (Default|CallPlaySequence))

(event Custom|ReplicatePlaySlab
  (Default|CallPlaySequence))

(event EventBeginPlay
  (Networking|SwitchHasAuthority
    (:Authority
      (Utilities|FlowControl|Sequence
        (:then_0
          (MCPContext|WatchCalendarEvent "R0"
            (:OnEventActive
              (if (not (Variables|Default|GetLevelStateSetup))
                (Variables|Default|SetSlabPOIChanged 0)
                (Variables|Default|SetLevelStateSetup true)))))
        (:then_1
          (MCPContext|WatchCalendarEvent "R1"
            (:OnEventActive
              (if (not (Variables|Default|GetLevelStateSetup))
                (Variables|Default|SetSlabPOIChanged 1)
                (Variables|Default|SetLevelStateSetup true)))))))))

(event Custom|SmashFromSequence
  (Default|CallSmashBuilding))

(event Custom|EventFromSequence
  (Default|CallLevelShouldbeHidden false))

(event Custom|SpawnSlabFromSequence)

(event EventTick
  (params (DeltaSeconds :type Float%20%28single-precision%29)))

(event Collision|EventActorBeginOverlap
  (params (OtherActor :type Actor%20Object%20Reference)))

(fn OnRep_SlabPOIChanged
  (switch int (Variables|Default|GetSlabPOIChanged)
    (:0
      (Default|CallLevelShouldbeHidden false))
    (:1
      (Default|CallLevelShouldbeHidden true))))

(fn SquashPlayers
  (params (OverlapActor :type Actor%20Object%20Reference))
  (for _array_element (Collision|GetOverlappingActors OverlapActor "/Game/Athena/PlayerPawn_Athena.PlayerPawn_Athena_C")
    (Environment|ApplyGlobalEnvironmentDamagetoActor _array_element 1e+03 "(GameplayTags=((TagName=\"Athena.Gameplay.Status.ForceSkipDBNO\"),(TagName=\"Gameplay.Damage.Environment\")))")))

(fn ConstructionScript)
