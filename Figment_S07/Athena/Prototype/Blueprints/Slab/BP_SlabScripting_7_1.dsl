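; Season 7 Slab scripting actor that picks which Slab POI state is active in Athena matches.
; After playlist data loads, the authority takes the first active slab1-4 calendar event as an index, unless a hotfix or debug override sets it.
; The SlabPOIChanged RepNotify broadcasts OnSlabLevelSelected with that index.
; A SquashPlayers helper deals 1000 environment damage, tagged to skip DBNO, to player pawns overlapping a given actor.
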(event EventBeginPlay
  (bind _asfort_game_state_athena (Utilities|Casting|CastToFortGameStateAthena (Game|GetGameState)))
  (if (Variables|/Script/FortniteGame.FortGameStateAthena|GetbPlaylistDataIsLoaded _asfort_game_state_athena)
    (Networking|SwitchHasAuthority
      (:Authority
        (if (== (ScalableFloat|GetValueatLevel (Variables|Default|GetHotfixSlabDisabled)) 1.0)
          (Variables|Default|SetSlabPOIChanged 4)
          (elif (== (Variables|Default|GetDebugSelectPOI) -1)
            (Utilities|FlowControl|Sequence
              (:then_0
                (MCPContext|WatchCalendarEvent "slab1"
                  (:OnEventActive
                    (if (not (Variables|Default|GetLevelStateSetup))
                      (Variables|Default|SetSlabPOIChanged 0)
                      (Variables|Default|SetLevelStateSetup true)))))
              (:then_1
                (MCPContext|WatchCalendarEvent "slab2"
                  (:OnEventActive
                    (if (not (Variables|Default|GetLevelStateSetup))
                      (Variables|Default|SetSlabPOIChanged 1)
                      (Variables|Default|SetLevelStateSetup true)))))
              (:then_2
                (MCPContext|WatchCalendarEvent "slab3"
                  (:OnEventActive
                    (if (not (Variables|Default|GetLevelStateSetup))
                      (Variables|Default|SetSlabPOIChanged 2)
                      (Variables|Default|SetLevelStateSetup true)))))
              (:then_3
                (MCPContext|WatchCalendarEvent "slab4"
                  (:OnEventActive
                    (if (not (Variables|Default|GetLevelStateSetup))
                      (Variables|Default|SetSlabPOIChanged 3)
                      (Variables|Default|SetLevelStateSetup true))))))
            (else
              (Variables|Default|SetSlabPOIChanged (Variables|Default|GetDebugSelectPOI)))))))
    (else
      (Playlist|BindEventtoOnCurrentPlaylistLoaded _asfort_game_state_athena (delegate Custom|ContinueSetup)))))

(event Custom|ContinueSetup
  (params (PlaylistName :type Name) (PlaylistContextTags :type Gameplay%20Tag%20Container%20Structure%20%28by%20ref%29))
  (Networking|SwitchHasAuthority
    (:Authority
      (if (== (ScalableFloat|GetValueatLevel (Variables|Default|GetHotfixSlabDisabled)) 1.0)
        (Variables|Default|SetSlabPOIChanged 4)
        (elif (== (Variables|Default|GetDebugSelectPOI) -1)
          (Utilities|FlowControl|Sequence
            (:then_0
              (MCPContext|WatchCalendarEvent "slab1"
                (:OnEventActive
                  (if (not (Variables|Default|GetLevelStateSetup))
                    (Variables|Default|SetSlabPOIChanged 0)
                    (Variables|Default|SetLevelStateSetup true)))))
            (:then_1
              (MCPContext|WatchCalendarEvent "slab2"
                (:OnEventActive
                  (if (not (Variables|Default|GetLevelStateSetup))
                    (Variables|Default|SetSlabPOIChanged 1)
                    (Variables|Default|SetLevelStateSetup true)))))
            (:then_2
              (MCPContext|WatchCalendarEvent "slab3"
                (:OnEventActive
                  (if (not (Variables|Default|GetLevelStateSetup))
                    (Variables|Default|SetSlabPOIChanged 2)
                    (Variables|Default|SetLevelStateSetup true)))))
            (:then_3
              (MCPContext|WatchCalendarEvent "slab4"
                (:OnEventActive
                  (if (not (Variables|Default|GetLevelStateSetup))
                    (Variables|Default|SetSlabPOIChanged 3)
                    (Variables|Default|SetLevelStateSetup true))))))
          (else
            (Variables|Default|SetSlabPOIChanged (Variables|Default|GetDebugSelectPOI))))))))

(fn OnRep_SlabPOIChanged
  (Default|CallOnSlabLevelSelected (Variables|Default|GetSlabPOIChanged)))

(fn SquashPlayers
  (params (OverlapActor :type Actor%20Object%20Reference))
  (for _array_element (Collision|GetOverlappingActors OverlapActor "/Game/Athena/PlayerPawn_Athena.PlayerPawn_Athena_C")
    (Environment|ApplyGlobalEnvironmentDamagetoActor _array_element 1e+03 "(GameplayTags=((TagName=\"Athena.Gameplay.Status.ForceSkipDBNO\"),(TagName=\"Gameplay.Damage.Environment\")))")))

(fn ConstructionScript)
